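# Physics asset for the SureBamboo two-handed pickaxe cosmetic mesh, giving it a simple collision shape.
# It holds a single body on the Melee_TwoHanded_Handle bone, built from one capsule offset from the bone and rotated to cover the axe.
# No constraints are defined, so the asset has no jointed or multi-body simulation.
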PhysicsAsset /BRCosmetics/Weapons/FORT_Melee/Pickaxe_SureBamboo/Meshes/SureBamboo_Axe_PhysicsAsset
bodies {Melee_TwoHanded_Handle [{shapeType=Capsule, length=74.5804, radius=34.4774, center=<0.6597 3.2946 29.4023>, rotation=<rot -0.5752 -90 0.0072>}]}
constraints []
